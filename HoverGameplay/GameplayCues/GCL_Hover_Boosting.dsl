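; Looping gameplay cue that plays audio and Niagara effects while a Fortnite player pawn is hover boosting.
; On becoming relevant, it attaches boost VFX to the backpack mesh and fades in looping audio on the pawn root.
; A repeating timer feeds the fuel attribute into the FuelFlicker parameter, retrying the VFX spawn if its component is missing.
; Ceasing relevance plays a guarded one-shot ended sound, then clears audio, effects and timer.
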(event Custom|StartUpdateFx
  ;[1 0 0.75] Update Niagara System
  (bind _boostfxcomponent (Variables|Fx|Component|GetBoostFxComponent)
    (:then
      (Effects|Components|ParticleSystem|SetFloatParameter _boostfxcomponent "FuelFlicker" (Math|Float|SafeDivide (Fuel|GetFuel) 1e+02)))
    (:else
      (Fx|TrySpawnBoostFx) ; In case backpack mesh was not valid
    )))

(fn CacheAndInitVariables
  (params (MyTarget :type Actor%20Object%20Reference))
  (Utilities|FlowControl|Sequence
    (:then_0
      (bind _asfort_player_pawn (Utilities|Casting|CastToFortPlayerPawn (Variables|GetMyTarget)))
      (Variables|Default|SetTargetPawn _asfort_player_pawn))
    (:then_1
      (Variables|Default|SetHasPlayedEndedSound false))))

(fn Cleanup
  (Utilities|FlowControl|Sequence
    (:then_0
      ; Stop Looping Audio
      (bind _boostingaudiocomponent (Variables|Audio|Component|GetBoostingAudioComponent)
        (:then
          (Audio|Components|Audio|FadeOut _boostingaudiocomponent 0.1 0.0 "SCurve" :!class /Script/Engine.AudioComponent)
          (Variables|Audio|Component|SetBoostingAudioComponent 0))
        (:else)))
    (:then_1
      ; Stop Looping Fx
      (bind _boostfxcomponent (Variables|Fx|Component|GetBoostFxComponent)
        (:then
          (Components|DestroyComponent _boostfxcomponent 0)
          (Variables|Fx|Component|SetBoostFxComponent 0))
        (:else)))
    (:then_2
      ; Clear timer and pawn
      (Utilities|Time|ClearandInvalidateTimerbyHandle self (Variables|Fx|GetUpdateFxHandle))
      (Variables|Default|SetTargetPawn 0)
      (Variables|Default|SetHasPlayedEndedSound false))))

(fn GetFuel
  (bind _targetpawn (Variables|Default|GetTargetPawn)
    (:then
      (return (GadgetAttributeSets|GetSimulatedAttributeValue _targetpawn (Variables|Fuel|GetFuelAttribute) 0.0 :!class /Script/FortniteGame.FortPlayerPawn)))
    (:else
      (return 0.0))))

(fn OnCeaseRelevant
  (params (MyTarget :type Actor%20Object%20Reference) (Parameters :type Gameplay%20Cue%20Parameters%20Structure%20%28by%20ref%29))
  (CallFunction|SpawnEndedSound)
  (CallFunction|Cleanup)
  (bind _returnvalue (|CallParentFunction (Variables|GetMyTarget) (Variables|GetParameters)))
  (return _returnvalue))

(fn SpawnEndedSound
  (bind _targetpawn (Variables|Default|GetTargetPawn))
  (bind _boostendedsound (Variables|Audio|GetBoostEndedSound))
  (Variables|Audio|Component|GetBoostingEndedAudioComponent
    (:then)
    (:else
      (if (Math|Boolean|ANDBoolean (Utilities|IsValid _boostendedsound :!class /Script/Engine.KismetSystemLibrary) (Utilities|IsValid _targetpawn :!class /Script/Engine.KismetSystemLibrary) (not (Variables|Default|GetHasPlayedEndedSound)))
        (bind _returnvalue (Audio|SpawnSoundAttached :Sound _boostendedsound :AttachToComponent (Variables|/Script/Engine.Actor|GetRootComponent _targetpawn) :AttachPointName "None" :LocationType "KeepRelativeOffset" :bStopWhenAttachedToDestroyed true :VolumeMultiplier 1.0 :PitchMultiplier 1.0 :StartTime 0.0 :bAutoDestroy true))
        (Variables|Audio|Component|SetBoostingEndedAudioComponent _returnvalue)
        (Variables|Default|SetHasPlayedEndedSound true)))))

(fn TrySpawnBoostAudio
  (bind _targetpawn (Variables|Default|GetTargetPawn))
  (bind _boostloopingsound (Variables|Audio|GetBoostLoopingSound))
  ; Start Audio
  (Variables|Audio|Component|GetBoostingAudioComponent
    (:then
      _)
    (:else
      (if (and (Utilities|IsValid _boostloopingsound :!class /Script/Engine.KismetSystemLibrary) (Utilities|IsValid _targetpawn :!class /Script/Engine.KismetSystemLibrary))
        (bind _returnvalue (Audio|SpawnSoundAttached _boostloopingsound (Variables|/Script/Engine.Actor|GetRootComponent _targetpawn)))
        (bind _output_get (Variables|Audio|Component|SetBoostingAudioComponent _returnvalue)))))
  (Audio|Components|Audio|FadeIn _output_get 0.5 1.0 0.0 "Linear" :!class /Script/Engine.AudioComponent))

(fn TrySpawnBoostFx
  (Variables|Fx|Component|GetBoostFxComponent
    (:then)
    (:else
      (bind _targetpawn (Variables|Default|GetTargetPawn)
        (:then
          (bind _output_get (LocalVariables|SetBackpackMesh (Customization|GetSkeletalMeshforPartType _targetpawn "Backpack" :!class /Script/FortniteGame.FortPlayerPawn)))
          (Utilities|IsValid _output_get
            (:Is%20Valid
              (bind _boostvfxsystem (Variables|Fx|GetBoostVfxSystem)
                (:then
                  (bind _returnvalue (Niagara|SpawnSystemAttached _boostvfxsystem _output_get "None" "0, 0, 0" "0, 0, 0" "KeepRelativeOffset" true))
                  (Variables|Fx|Component|SetBoostFxComponent _returnvalue))
                (:else)))))
        (:else)))))

(fn ConstructionScript)

(fn OnBecomeRelevant
  (params (MyTarget :type Actor%20Object%20Reference) (Parameters :type Gameplay%20Cue%20Parameters%20Structure%20%28by%20ref%29))
  (CallFunction|CacheandInitVariables MyTarget)
  (Fx|TrySpawnBoostFx)
  (Audio|TrySpawnBoostAudio)
  (CallFunction|StartUpdateFx)
  (bind _returnvalue (Utilities|Time|SetTimerbyEvent (EventDispatchers|CreateEvent :!event StartUpdateFx) (Variables|Fx|GetBoostFxUpdateRate) true true))
  (Variables|Fx|SetUpdateFxHandle _returnvalue)
  (bind _returnvalue_2 (|CallParentFunction (Variables|GetMyTarget) (Variables|GetParameters)))
  (return _returnvalue_2))
