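; Looping gameplay cue for hover fuel regeneration that plays an attached sound when regeneration ends.
; OnBurst and OnBecomeRelevant cache the target as a Fort player pawn.
; OnCeaseRelevant plays the end sound only for a locally viewed pawn, then clears cached references.
; The simulated fuel attribute picks the full or not-full sound, with the latter played only if AllowNotFullSound is set.
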(event EventBeginPlay)

(event Collision|EventActorBeginOverlap
  (params (OtherActor :type Actor%20Object%20Reference)))

(event EventTick
  (params (DeltaSeconds :type Float%20%28single-precision%29)))

(fn CacheVariables
  (params (MyTarget :type Actor%20Object%20Reference))
  (bind _asfort_player_pawn (Utilities|Casting|CastToFortPlayerPawn MyTarget))
  (Variables|Default|SetTargetPawn _asfort_player_pawn))

(fn Cleanup
  (Variables|Audio|Component|SetFuelRegenEndedAudioComponent 0)
  (Variables|Default|SetTargetPawn 0))

(fn GetAttachComponent
  (bind _targetpawn (Variables|Default|GetTargetPawn)
    (:then
      (return (Variables|/Script/Engine.Actor|GetRootComponent _targetpawn)))
    (:else
      (return 0))))

(fn GetFuelPercent
  (bind _targetpawn (Variables|Default|GetTargetPawn)
    (:then
      (return (Math|Float|SafeDivide (GadgetAttributeSets|GetSimulatedAttributeValue _targetpawn (Variables|Fuel|GetFuelAttribute) 0.0 :!class /Script/FortniteGame.FortPlayerPawn) 1e+02)))
    (:else
      (return 0.0))))

(fn OnBurst
  (params (MyTarget :type Actor%20Object%20Reference) (Parameters :type Gameplay%20Cue%20Parameters%20Structure%20%28by%20ref%29))
  (CallFunction|CacheVariables MyTarget)
  (bind _returnvalue (|CallParentFunction (Variables|GetMyTarget) (Variables|GetParameters)))
  (return _returnvalue))

(fn OnCeaseRelevant
  (params (MyTarget :type Actor%20Object%20Reference) (Parameters :type Gameplay%20Cue%20Parameters%20Structure%20%28by%20ref%29))
  (Audio|SpawnEndedSound)
  (CallFunction|Cleanup)
  (bind _returnvalue (|CallParentFunction (Variables|GetMyTarget) (Variables|GetParameters)))
  (return _returnvalue))

(fn SpawnEndedSound
  (bind _targetpawn (Variables|Default|GetTargetPawn)
    (:then
      (if (Pawn|IsLocallyViewed _targetpawn)
        ; Always play sound for full. Or play either sound if we allow not full sound
        (bind _output_get (LocalVariables|SetIsFuelFull (Math|Float|NearlyEqual(Float) (Fuel|GetFuelPercent) 1.0 0.02)))
        (if (or _output_get (Variables|Default|GetAllowNotFullSound))
          (LocalVariables|SetDesiredSound (select (LocalVariables|GetIsFuelFull) (Variables|Audio|GetFuelRegenEndedFullSound) (Variables|Audio|GetFuelRegenEndedNotFullSound)))
          (bind _desiredsound (LocalVariables|GetDesiredSound)
            (:then
              (bind _returnvalue (Audio|SpawnSoundAttached _desiredsound (CallFunction|GetAttachComponent)))
              (Variables|Audio|Component|SetFuelRegenEndedAudioComponent _returnvalue))
            (:else)))))
    (:else)))

(fn ConstructionScript)

(fn OnBecomeRelevant
  (params (MyTarget :type Actor%20Object%20Reference) (Parameters :type Gameplay%20Cue%20Parameters%20Structure%20%28by%20ref%29))
  (CallFunction|CacheVariables MyTarget)
  (bind _returnvalue (|CallParentFunction (Variables|GetMyTarget) (Variables|GetParameters)))
  (return _returnvalue))
