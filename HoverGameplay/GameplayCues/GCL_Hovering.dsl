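; Looping gameplay cue that plays the visual and audio effects for a pawn's backpack hover.
; Burst and relevancy events start the sounds and backpack FX, with timed updates retrying the FX until the mesh is valid.
; Parent-driven events feed height, velocity and fuel modulation, play fuel warnings, and move or toggle ground FX.
; Cleanup fades looping audio and releases manually pooled FX, while overrides limit sound indicators to other players' neutral or hostile pawns.
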(event EventUpdateFx
  ;[0.3552 1 0.3552] Update Fx - Called from Timer set in WhileActive
  (|CallParentFunction)
  ;;[1 1 0 1 12] Trying to spawn the hover looping fx as the backpack mesh may not be valid when a pawn/this gameplaycue comes into relevancy.
  (VFX|TrySpawnHoverFx))

(event EventUpdateAudioParameters
  (params (HeightMod :type Float%20%28single-precision%29) (VelocityMod :type Float%20%28single-precision%29))
  ;[0.009804 0.1028 0.15 0.5] If Hover Looping Audio Component is Valid and Is Playing then pass modulation data in
  (bind _hoverloopingaudiocomponent (Variables|Audio|GetHoverLoopingAudioComponent)
    (:then
      (if (Audio|Components|Audio|IsPlaying _hoverloopingaudiocomponent)
        (Audio|Parameter|SetFloatParameter _hoverloopingaudiocomponent (Variables|Audio|GetHeightModParameterName) HeightMod :!class /Script/AudioExtensions.AudioParameterControllerInterface)
        (Audio|Parameter|SetFloatParameter _hoverloopingaudiocomponent (Variables|Audio|GetVelocityModParameterName) VelocityMod :!class /Script/AudioExtensions.AudioParameterControllerInterface)))
    (:else)))

(event Fuel|EventUpdateFuelFlicker
  (params (FlickerPercent :type Float%20%28single-precision%29))
  ;[0.03526 0.15 0.07155 0.5] Update hover FX fuel flicker paramter with current fuel percentage or height limit override
  (bind _hoverfxcomponent (Variables|VFX|GetHoverFxComponent)
    (:then
      (Effects|Components|ParticleSystem|SetFloatParameter _hoverfxcomponent "FuelFlicker" FlickerPercent))
    (:else)))

(event Fuel|EventOnFuelLow
  ;[0.1299 0.05472 0.15 0.5] Play fuel low audio cue
  (Variables|Audio|GetLowFuelAudioComponent
    (:then)
    (:else
      (bind _returnvalue (Audio|SpawnSoundAttached (Variables|Audio|GetLowFuelSound) (Variables|/Script/Engine.Actor|GetRootComponent (Variables|Runtime|GetTargetPawn))))
      (Variables|Audio|SetLowFuelAudioComponent _returnvalue))))

(event Fuel|EventOnOutOfFuel
  ;[0.15 0.03967 0.13 0.5] Play out of fuel audio cue
  (Variables|Audio|GetOutOfFuelAudioComponent
    (:then)
    (:else
      (bind _outoffuelsound (Variables|Audio|GetOutOfFuelSound)
        (:then
          (bind _returnvalue (Audio|SpawnSoundAttached :Sound _outoffuelsound :AttachToComponent (Variables|/Script/Engine.Actor|GetRootComponent (Variables|Runtime|GetTargetPawn)) :AttachPointName "None" :LocationType "KeepRelativeOffset" :bStopWhenAttachedToDestroyed true :VolumeMultiplier 1.0 :PitchMultiplier 1.0 :StartTime 0.0 :bAutoDestroy true))
          (Variables|Audio|SetOutOfFuelAudioComponent _returnvalue))
        (:else)))) ; Only spawn if not valid
)

(event EventCleanup
  ;[0.15 0.04546 0.008902 0.5] Cleanup
  (|CallParentFunction)
  (Utilities|FlowControl|Sequence
    (:then_0
      ;; Turn everything off
      (Sound|EnableHoverAudio false)
      (Effects|Components|ParticleSystem|ReleasetoPool (Variables|VFX|GetHoverFxComponent))
      (Effects|Components|ParticleSystem|ReleasetoPool (Variables|VFX|GetGroundFxComponent)))
    (:then_1
      ;; Clear Audio component references
      (Variables|Audio|SetHoverLoopingAudioComponent 0)
      (Variables|Audio|SetStartHoverAudioComponent 0)
      (Variables|Audio|SetLowFuelAudioComponent 0)
      (Variables|Audio|SetOutOfFuelAudioComponent 0))
    (:then_2
      ;; Clear Vfx component references
      (Variables|VFX|SetHoverFxComponent 0)
      (Variables|VFX|SetGroundFxComponent 0))))

(fn ActivateGroundFx
  (params (Activate :type Boolean))
  (if Activate
    (bind _groundfxcomponent (Variables|VFX|GetGroundFxComponent)
      (:then
        (if (not (Components|Activation|IsActive _groundfxcomponent))
          (Components|Activation|Activate _groundfxcomponent true)))
      (:else
        ;[1 1 0] Try to spawn a new ground fx component
        (GroundFx|SpawnGroundFx)))
    (else
      (bind _groundfxcomponent_2 (Variables|VFX|GetGroundFxComponent)
        (:then
          (Components|Activation|Deactivate _groundfxcomponent_2))
        (:else)))))

(fn EnableHoverAudio
  (params (Enable :type Boolean))
  (bind _targetpawn (Variables|Runtime|GetTargetPawn))
  (bind _audiofadeoutduration (Variables|Audio|GetAudioFadeOutDuration))
  (if Enable
    ;[0 1 0] Spawn the looping hover audio - All other audio should be spawned based on events
    (if (and (not (Utilities|IsValid (Variables|Audio|GetHoverLoopingAudioComponent) :!class /Script/Engine.KismetSystemLibrary)) (Utilities|IsValid _targetpawn :!class /Script/Engine.KismetSystemLibrary))
      (bind _returnvalue (Audio|SpawnSoundAttached :Sound (Variables|Audio|GetHoverSound) :AttachToComponent (Variables|/Script/Engine.Actor|GetRootComponent _targetpawn) :AttachPointName "None" :LocationType "KeepRelativeOffset" :bStopWhenAttachedToDestroyed true :VolumeMultiplier 1.0 :PitchMultiplier 1.0 :StartTime 0.0 :bAutoDestroy true))
      (Variables|Audio|SetHoverLoopingAudioComponent _returnvalue))
    (else
      ;[1 0.25 0] Fade out audio which should stop when done and destroy the components
      (bind _hoverloopingaudiocomponent (Variables|Audio|GetHoverLoopingAudioComponent)
        (:then
          (Audio|Components|Audio|FadeOut _hoverloopingaudiocomponent _audiofadeoutduration 0.0 "Linear" :!class /Script/Engine.AudioComponent))
        (:else
          _))
      (bind _lowfuelaudiocomponent (Variables|Audio|GetLowFuelAudioComponent)
        (:then
          (Audio|Components|Audio|FadeOut _lowfuelaudiocomponent _audiofadeoutduration 0.0 "Linear" :!class /Script/Engine.AudioComponent))
        (:else)))))

(fn GetAffiliationsToShowFor
  (return (Utilities|Array|MakeArray "Neutral" "Hostile")))

(fn OnBurst
  (params (MyTarget :type Actor%20Object%20Reference) (Parameters :type Gameplay%20Cue%20Parameters%20Structure%20%28by%20ref%29))
  (bind _starthoversound (Variables|Audio|GetStartHoverSound))
  ;[0.04348 0.15 0.03092 0.5] Play start hover sound on start
  (bind _returnvalue (|CallParentFunction MyTarget Parameters))
  (LocalVariables|SetRetVal _returnvalue)
  (if (and (Utilities|IsValid _starthoversound :!class /Script/Engine.KismetSystemLibrary) (Utilities|IsValid MyTarget :!class /Script/Engine.KismetSystemLibrary))
    (bind _returnvalue_2 (Audio|SpawnSoundAttached :Sound _starthoversound :AttachToComponent (Variables|/Script/Engine.Actor|GetRootComponent MyTarget) :AttachPointName "None" :LocationType "KeepRelativeOffset" :bStopWhenAttachedToDestroyed true :VolumeMultiplier 1.0 :PitchMultiplier 1.0 :StartTime 0.0 :bAutoDestroy true))
    (Variables|Audio|SetStartHoverAudioComponent _returnvalue_2)
    (else
      _))
  (return (LocalVariables|GetRetVal)))

(fn ShouldShowSoundIndicator
  (params (PlayerController :type Fort%20Player%20Controller%20Object%20Reference))
  (bind _targetpawn (Variables|Runtime|GetTargetPawn)
    (:then
      (return (not (== _targetpawn (Pawn|GetControlledPawn PlayerController)))))
    (:else
      (return false))))

(fn SpawnGroundFx
  (Variables|VFX|GetGroundFxComponent
    (:then)
    (:else
      (bind _returnvalue (Effects|Components|ParticleSystem|SpawnFXAtLocation :WorldContextObject self :FXSystemAsset (Variables|VFX|GetGroundInteractionSystem) :Location (Variables|Runtime|GroundFX|GetGroundFXLocation) :Scale "1.000000,1.000000,1.000000" :bAutoDestroy false :PoolingMethod "ManualRelease" :bAutoActivate true))
      (Variables|VFX|SetGroundFxComponent _returnvalue))))

(fn TrySpawnHoverFx
  (if (Math|Boolean|ANDBoolean (Utilities|IsValid (Variables|Runtime|GetBackpackMesh) :!class /Script/Engine.KismetSystemLibrary) (not (Utilities|IsValid (Variables|VFX|GetHoverFxComponent) :!class /Script/Engine.KismetSystemLibrary)) (Utilities|IsValid (Variables|Runtime|GetTargetPawn) :!class /Script/Engine.KismetSystemLibrary))
    (bind _hoverfxsystem (Variables|VFX|GetHoverFxSystem)
      (:then
        (bind _returnvalue (Effects|SpawnFXSystemAttached :FXSystemAsset _hoverfxsystem :AttachToComponent (Variables|Runtime|GetBackpackMesh) :AttachPointName "None" :Scale "1.000000,1.000000,1.000000" :LocationType "KeepRelativeOffset" :bAutoDestroy false :PoolingMethod "ManualRelease" :bAutoActivate true)) ; Spawn
        (bind _output_get (Variables|VFX|SetHoverFxComponent _returnvalue))
        (Utilities|IsValid _output_get
          (:Is%20Valid
            (Effects|Components|ParticleSystem|SetBoolParameter _output_get "IsLocal" (Pawn|IsLocallyViewed (Variables|Runtime|GetTargetPawn))))))
      (:else))))

(fn UpdateGroundFxLocation
  (bind _groundfxheight (Variables|Runtime|GroundFX|GetGroundFXHeight))
  (bind _groundfxcomponent (Variables|VFX|GetGroundFxComponent)
    (:then
      (Transformation|SetWorldLocationAndRotation _groundfxcomponent (Variables|Runtime|GroundFX|GetGroundFXLocation) (Variables|Runtime|GroundFX|GetGroundFxRotation) false false)
      (if (> _groundfxheight 0.5)
        (Effects|Components|ParticleSystem|SetFloatParameter _groundfxcomponent "Height" _groundfxheight)))
    (:else)))

(fn UpdateGroundFX
  (bind _returnvalue (|CallParentFunction))
  (if _returnvalue
    (Utilities|FlowControl|Sequence
      (:then_0
        (GroundFx|SpawnGroundFx))
      (:then_1
        (GroundFx|UpdateGroundFxLocation))
      (:then_2
        (GroundFx|ActivateGroundFx true))
      (:then_3
        (return true)))
    (else
      (GroundFx|ActivateGroundFx false)
      (return false))))

(fn ConstructionScript)

(fn OnBecomeRelevant
  (params (MyTarget :type Actor%20Object%20Reference) (Parameters :type Gameplay%20Cue%20Parameters%20Structure%20%28by%20ref%29))
  (bind _returnvalue (|CallParentFunction MyTarget Parameters))
  (Utilities|FlowControl|Sequence
    (:then_0
      (VFX|TrySpawnHoverFx))
    (:then_1
      (Sound|EnableHoverAudio true))
    (:then_2
      (return _returnvalue))))
