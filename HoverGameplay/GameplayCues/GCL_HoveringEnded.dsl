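; Looping gameplay cue that plays backpack VFX and looping falling audio on a player pawn after hovering ends in the air.
; On becoming relevant it caches the pawn and backpack mesh, then starts a timer that polls the simulated fuel attribute.
; When fuel presence changes, it swaps between low-thrust and no-fuel sputter effects and sounds.
; Cleanup runs on cease relevant and on Removed events, the latter guarding against stuck loops.
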(event Custom|UpdateFuelAndFx
  (bind _returnvalue (CallFunction|GetFuel))
  (bind _returnvalue_2 (> _returnvalue 0))
  ;[1 1 0] We are watching fuel in case someone falls for a long time and fuel regens so we need to switch out the VFX
  (if (or (not (Variables|Default|GetCompletedFirstTimeUpdate)) (!= (Variables|Fuel|GetHasFuel) _returnvalue_2))
    (Variables|Fuel|SetHasFuel _returnvalue_2)
    (Systems|UpdateAudioandVisuals)))

(event GameplayCueNotify|EventHandleGameplayCue
  (params (MyTarget :type Actor%20Object%20Reference) (EventType :type EGameplayCueEvent%20Enum :default "OnActive") (Parameters :type Gameplay%20Cue%20Parameters%20Structure%20%28by%20ref%29))
  ;[1 0.25 0] Fix for loop getting stuck
  (Utilities|IsValid MyTarget
    (:Is%20Valid
      (if (Utilities|Enum|Equal(Enum) EventType "Removed")
        (CallFunction|Cleanup)))))

(fn CacheVariables
  (params (InTarget :type Actor%20Object%20Reference))
  (bind _asfort_player_pawn (Utilities|Casting|CastToFortPlayerPawn InTarget))
  (bind _output_get (Variables|Default|SetTargetPawn _asfort_player_pawn))
  (Variables|Default|SetBackPackMeshComponent (Customization|GetSkeletalMeshforPartType _output_get "Backpack" :!class /Script/FortniteGame.FortPlayerPawn))
  (Variables|Fuel|SetFuel (CallFunction|GetFuel)))

(fn Cleanup
  (Utilities|FlowControl|Sequence
    (:then_0
      ; Deactivate VFX system
      (bind _endhoverinairfxcomponent (Variables|Default|GetEndHoverInAirFXComponent)
        (:then
          (Components|Activation|Deactivate _endhoverinairfxcomponent)
          (Variables|Default|SetEndHoverInAirFXComponent 0))
        (:else)))
    (:then_1
      ; Clear timer
      (Utilities|Time|ClearandInvalidateTimerbyHandle self (Variables|Fuel|GetWatchFuelHandle))
      (Variables|Default|SetCompletedFirstTimeUpdate false)
      (Variables|Default|SetBackPackMeshComponent 0)
      (Variables|Default|SetTargetPawn 0))
    (:then_2
      ; Fade out audio
      (bind _whilefallingloopingaudiocomponent (Variables|Audio|Component|GetWhileFallingLoopingAudioComponent)
        (:then
          (Audio|Components|Audio|FadeOut _whilefallingloopingaudiocomponent 0.1 0.0 "Linear" :!class /Script/Engine.AudioComponent)
          (Variables|Audio|Component|SetWhileFallingLoopingAudioComponent 0))
        (:else)))))

(fn GetFuel
  (bind _targetpawn (Variables|Default|GetTargetPawn)
    (:then
      (return (GadgetAttributeSets|GetSimulatedAttributeValue _targetpawn (Variables|Fuel|GetFuelAttribute) (LocalVariables|GetDefaultValue) :!class /Script/FortniteGame.FortPlayerPawn)))
    (:else
      (return (LocalVariables|GetDefaultValue)))))

(fn OnCeaseRelevant
  (params (MyTarget :type Actor%20Object%20Reference) (Parameters :type Gameplay%20Cue%20Parameters%20Structure%20%28by%20ref%29))
  (CallFunction|Cleanup)
  (bind _returnvalue (|CallParentFunction MyTarget Parameters))
  (return _returnvalue))

(fn UpdateAudioAndVisuals
  (bind _desiredsound (LocalVariables|GetDesiredSound))
  (Utilities|FlowControl|Sequence
    (:then_0
      ;[1 0 0.75] Handle Visuals
      ;; Destroy the current system
      (bind _endhoverinairfxcomponent (Variables|Default|GetEndHoverInAirFXComponent)
        (:then
          (Components|DestroyComponent _endhoverinairfxcomponent 0))
        (:else
          _))
      ;; Spawn new system based on fuel
      (bind _backpackmeshcomponent (Variables|Default|GetBackPackMeshComponent)
        (:then
          (bind _output_get (LocalVariables|SetDesiredFxSystem (select (Variables|Fuel|GetHasFuel) (Variables|Systems|GetHoverEndInAirFX) (Variables|Systems|GetHoverEndInAirNoFuelFX))))
          (Utilities|IsValid _output_get
            (:Is%20Valid
              (bind _returnvalue (Effects|SpawnFXSystemAttached _output_get _backpackmeshcomponent))
              (bind _output_get_2 (Variables|Default|SetEndHoverInAirFXComponent _returnvalue))
              (Effects|Components|ParticleSystem|SetFloatParameter _output_get_2 "FuelFlicker" (Math|Float|SelectFloat 1.0))
              (Variables|Default|SetCompletedFirstTimeUpdate true))))
        (:else)))
    (:then_1
      (Variables|Default|GetTargetPawn
        (:then
          (bind _output_get_3 (LocalVariables|SetDesiredSound (select (Variables|Fuel|GetHasFuel) (Variables|Audio|GetWhileFallingLowThrustSound) (Variables|Audio|GetWhileFallingSputterSound))))
          (Utilities|IsValid _output_get_3
            (:Is%20Valid
              ;[1 1 0] If valid set the sound - needs testing
              (bind _whilefallingloopingaudiocomponent (Variables|Audio|Component|GetWhileFallingLoopingAudioComponent)
                (:then
                  (Audio|Components|Audio|SetSound _whilefallingloopingaudiocomponent _desiredsound))
                (:else
                  ;[0.2 0.2 1] If not valid spawn an audio component
                  (bind _returnvalue_2 (Audio|SpawnSoundAttached _desiredsound (Variables|/Script/Engine.Actor|GetRootComponent (Variables|Default|GetTargetPawn))))
                  (Variables|Audio|Component|SetWhileFallingLoopingAudioComponent _returnvalue_2))))))
        (:else)))))

(fn ConstructionScript)

(fn OnBecomeRelevant
  (params (MyTarget :type Actor%20Object%20Reference) (Parameters :type Gameplay%20Cue%20Parameters%20Structure%20%28by%20ref%29))
  (CallFunction|CacheVariables MyTarget)
  ;[1 0 0.75] Called initially and then we continue to watch the fuel value in case regen kicks in before this gcl is cancelled. In this case we want to update from sputtering to low thrust fx
  (CallFunction|UpdateFuelandFx)
  (bind _returnvalue (Utilities|Time|SetTimerbyEvent (EventDispatchers|CreateEvent :!event UpdateFuelAndFx) (ScalableFloat|GetValueatLevel (Variables|Default|GetFxUpdateRate)) true true))
  (Variables|Fuel|SetWatchFuelHandle _returnvalue)
  (bind _returnvalue_2 (|CallParentFunction (Variables|GetMyTarget) (Variables|GetParameters)))
  (return _returnvalue_2))
