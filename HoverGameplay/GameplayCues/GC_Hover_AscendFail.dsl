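; Burst gameplay cue, named for a failed hover ascend, that plays a Niagara effect on the player's backpack.
; OnBurst calls SpawnVfxSystem on the target, then runs the parent burst logic.
; SpawnVfxSystem casts the target to FortPlayerPawn, fetches its Backpack skeletal mesh and, when mesh and VfxSystem are valid, attaches the system at "root".
; A debug PrintString reports the spawned component's name.
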(fn OnBurst(EitherVFX-preferthisversion)
  (params (MyTarget :type Actor%20Object%20Reference) (Parameters :type Gameplay%20Cue%20Parameters%20Structure%20%28by%20ref%29) (ParticleComponents :type Array%20of%20FXSystem%20Component%20Object%20References) (AudioComponents :type Array%20of%20Audio%20Component%20Object%20References) (BurstCameraShakeInstance :type Camera%20Shake%20Base%20Object%20Reference) (BurstDecalInstance :type Decal%20Actor%20Object%20Reference))
  (CallFunction|SpawnVfxSystem MyTarget)
  (|CallParentFunction MyTarget Parameters ParticleComponents AudioComponents BurstCameraShakeInstance BurstDecalInstance)
  (return))

(fn SpawnVfxSystem
  (params (AttachActor :type Actor%20Object%20Reference))
  (bind _asfort_player_pawn (Utilities|Casting|CastToFortPlayerPawn AttachActor))
  (bind _output_get (LocalVariables|SetBackpackMesh (Customization|GetSkeletalMeshforPartType _asfort_player_pawn "Backpack" :!class /Script/FortniteGame.FortPlayerPawn)))
  (Utilities|IsValid _output_get
    (:Is%20Valid
      (bind _vfxsystem (Variables|Default|GetVfxSystem)
        (:then
          (bind _returnvalue (Niagara|SpawnSystemAttached _vfxsystem _output_get "root" "0, 0, 0" "0, 0, 0" "KeepRelativeOffset" true))
          (Development|PrintString self (Utilities|String|Append "SystemSpawned = ( " (Utilities|GetDisplayName _returnvalue) " )")))
        (:else)))))
